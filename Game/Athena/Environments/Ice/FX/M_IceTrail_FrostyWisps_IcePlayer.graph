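# Translucent lit Niagara sprite material for frosty mist wisps, apparently the Ice Player's trail.
# An 8x8 SubUV mist flipbook plus panned tiling dust, tinted by particle color and driven by dynamic parameters, feeds base color and emissive.
# Optional controls remove dust by camera angle and apply time-of-day shading to emissive.
# Opacity uses depth and near-camera fades, skipped on mobile, plus optional Slurp Vat mask and sine-based edge softening.
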Material /Game/Athena/Environments/Ice/FX/M_IceTrail_FrostyWisps_IcePlayer
graph {
 properties {
  bUsedWithNanite=true
  bUsedWithNiagaraSprites=true
  blendMode=BLEND_Translucent
  pixelDepthOffsetMode=PDOM_Legacy
  translucencyLightingMode=TLM_VolumetricPerVertexNonDirectional
  usedShadingModels=Default Lit
 }
 nodes [
  n1 = CurveAtlasRowParameter(CurveTime=CollectionParameter() {collection=&FortniteMaterialParameters, parameterName=TimeOfDay_0-1}) {
   atlas=&AthenaCurveAtlas01
   curve=/Game/Athena/LinearColorCurves/Curve_TimeOfDay_S23
   defaultValue=321
   group=TimeOfDay
   parameterName=TOD_Curve
  }
  n2 = ScalarParameter() {defaultValue=1, group=Color, parameterName="Value_Morning (S)", sortPriority=4}
  n3 = ScalarParameter() {defaultValue=1, group=Color, parameterName="Value_Day (S)", sortPriority=3}
  n4 = ScalarParameter() {defaultValue=1, group=Color, parameterName="Value_Evening (S)", sortPriority=6}
  n5 = ScalarParameter() {defaultValue=1, group=Color, parameterName="Value_Night (S)", sortPriority=5}
  n6 = ParticleColor()
  n7 = TextureSampleParameterSubUV!(UVs, Apply View MipBias) {
   group=Texture
   parameterName=Cloud SubUV
   samplerType=SAMPLERTYPE_LinearColor
   texture=/Game/Effects/Fort_Effects/Textures/Smoke/T_FallingMist_8x8_01
  }
  n8 = LinearInterpolate(A=ScalarParameter() {parameterName=SubUVMinBrightness}, Alpha=@n7.R)
  n9 = DynamicParameter() {defaultValue=<rgba 1 1 1 0>, paramNames [DustTiling, Dust Emissive, Cloud Density, Panning MacroUV Control]}
  n10 = Multiply(A=@n8, B=@n9.Cloud Density)
  n11 = ScalarParameter() {group=Texture, parameterName=Cloud Density DynParam Control, sortPriority=1}
  n12 = LinearInterpolate(A=ParticleMacroUV(), B=TextureCoordinate(), Alpha=ScalarParameter() {group=Texture, parameterName=Use MacroUV, sortPriority=4})
  n13 = DynamicParameter() {defaultValue=<rgba 1 1 1 0>, paramNames [DustTiling, Dust Emissive, Cloud Density, Panning MacroUV Control]}
  n14 = Time()
  n15 = ScalarParameter() {group=Texture, parameterName=U_Speed, sortPriority=5}
  n16 = ScalarParameter() {group=Texture, parameterName=V_Speed, sortPriority=6}
  n17 = TextureSampleParameter2D(UVs=Panner(Coordinate=Multiply(A=@n12, B=@n13.DustTiling), Time=Multiply(A=@n13.Panning MacroUV Control, B=@n14), Speed=AppendVector(A=@n15, B=@n16))) {group=Texture, parameterName=Tiling Dust Texture, sortPriority=2, texture=/Game/Global/Textures/Default/PointSplatter/T_Tiling_Dust}
  n18 = Multiply(A=@n9.Dust Emissive, B=@n17.RGB)
  n19 = ScalarParameter() {group=Texture, parameterName=Dust Emissive DynParam Control, sortPriority=3}
  n20 = Multiply(A=@n6.RGB, B=Add(A=LinearInterpolate(A=@n8, B=@n10, Alpha=@n11), B=LinearInterpolate(A=@n17.RGB, B=@n18, Alpha=@n19)))
  n21 = CameraPositionWS()
  n22 = VectorParameter() {group=Angle Based Dust Removal, parameterName=WorldDirection, sortPriority=2, defaultValue=<rgba 0 0 1 0>}
  n23 = Transform(None=@n22.RGB) {transformSourceType=TRANSFORMSOURCE_Local}
  n24 = Power(
   Base=DotProduct(A=Normalize(VectorInput=@n21), B=Normalize(VectorInput=@n23))
   Exp=ScalarParameter() {defaultValue=4.5, group=Angle Based Dust Removal, parameterName=Camera Direction Fade Power, sortPriority=1}
  ) {constExponent=3}
  n25 = LinearInterpolate(
   A=@n20
   B=LinearInterpolate(A=@n20, B=Multiply(A=@n10, B=@n6.RGB), Alpha=Clamp(None=@n24))
   Alpha=ScalarParameter() {group=Angle Based Dust Removal, parameterName=Remove Dust Based on Angle}
  )
  n26 = Multiply(A=@n25, B=VectorParameter() {group=Color, parameterName=EmissiveTweak, sortPriority=1, defaultValue=<rgba 2 2 2 0>}.RGB) {constB=0.5}
  n27 = LinearInterpolate(
   A=@n26
   B=Multiply(
    A=@n26
    B=MaterialFunctionCall(Value_Morning=@n2, Value_Day=@n3, Value_Evening=@n4, Value_Night=@n5) {materialFunction=&Scalar_Time_of_Day}.Result
   )
   Alpha=ScalarParameter() {group=Color, parameterName=Use TOD Shading, sortPriority=2}
  )
  MaterialFunctionCall(Emissive=StaticSwitchParameter(True=Multiply(A=@n1, B=@n27), False=@n27) {parameterName=Use_S23ToD_Curve}) {materialFunction=&MF_FX_EmissiveAdjustments}
  n28 = TextureSampleParameterSubUV() {
   group=Texture
   parameterName=Cloud SubUV
   samplerType=SAMPLERTYPE_LinearColor
   texture=&T_FallingMist_8x8_01
  }
  n29 = Multiply(A=@n28.B, B=ParticleColor().A)
  n30 = MaterialFunctionCall(Distance=ScalarParameter() {defaultValue=500, group=Opacity, parameterName=Near Camera Fade Distance, sortPriority=2}) {materialFunction=&Near_Camera_Fade}
  n31 = ShadingPathSwitch(
   Default=DepthFade(
    Opacity=Multiply(A=@n29, B=@n30.Result)
    FadeDistance=ScalarParameter() {defaultValue=25, group=Opacity, parameterName=Depth Fade Distance, sortPriority=1}
   )
   Mobile=@n29
  )
  n32 = MaterialFunctionCall() {materialFunction=/Game/Athena/Items/Gameplay/SilkyVeteran/FX/SlurpVatMask}
  n33 = LinearInterpolate(A=@n32.Result, B=OneMinus(None=@n32.Result), Alpha=ScalarParameter() {group=Opacity, parameterName=InvertS lurp Vat Mask})
  n34 = Multiply(A=Multiply(A=TextureCoordinate()) {constB=8}) {constB=0.5}
  n35 = MaterialFunctionCall(Float2=Abs(None=Sine(None=@n34))) {materialFunction=&BreakOutFloat2Components}
  Multiply(
   A=LinearInterpolate(
    A=@n31
    B=Multiply(A=@n31, B=@n33)
    Alpha=ScalarParameter() {group=Opacity, parameterName="Multiply by World Sphere Mask (Slurp Vat)", sortPriority=3}
   )
   B=LinearInterpolate(B=Multiply(A=@n35.R, B=@n35.R), Alpha=ScalarParameter() {group=Opacity, parameterName=SoftenEdge}) {constA=1}
  )
  Reroute(None=Multiply(A=ScalarParameter() {defaultValue=1, group=Color, parameterName=BaseColorIntensity}, B=@n25))
 ]
}
